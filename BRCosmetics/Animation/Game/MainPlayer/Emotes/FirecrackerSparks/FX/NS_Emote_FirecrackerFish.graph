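# Niagara system for the firecracker sparks emote that flashes a glowing orange fish-shaped mesh at a fixed offset on a timed delay.
# The single fish mesh jitters and grows through a scale curve while fading out, with a dynamic material parameter animated early in its life.
# Four staggered sprite spark bursts with drag and gravity fire around it, three in angled cones and the last scattering wider, alongside a late soft glow sprite.
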NiagaraSystem /BRCosmetics/Animation/Game/MainPlayer/Emotes/FirecrackerSparks/FX/NS_Emote_FirecrackerFish
dependencies {
 usedDataInterfaces [
  &NiagaraDataInterfaceCurve
  &NiagaraDataInterfaceVectorCurve
  &NiagaraDataInterfaceMeshRendererInfo
 ]
 usedDynamicInputs []
 usedModules [&SystemState]
}
emitters {
 Fish_Shape {
  emitterUpdateScript [
   &EmitterState()
   &SpawnRate(SpawnRate=40, Use Spawn Probability=false) {enabled=false}
   &SpawnBurst_Instantaneous(Spawn Count=1, Spawn Time=1.4, Use Spawn Probability=false)
  ]
  particleSpawnScript [
   &InitializeParticle(Lifetime=2, Mesh Scale Mode=::NewEnumerator3, Mesh Uniform Scale=0.2, UsePositionOffset=false, Write Material Random=true) {module=InitializeParticle001}
   /Game/Effects/Niagara/Modules/Orientation/FN_Fan(
    Cone Angle:NiagaraFloat=&UniformRangedFloat(Minimum=0, Maximum=0)
    Spin:NiagaraFloat=&UniformRangedFloat(Minimum=0, Maximum=6.2)
   ) {enabled=false}
   &SphereLocation(
    MaskSpawnGroup=false
    Offset=<70 120 120>
    Override Seed=false
    Sphere Radius=0.01
    Surface Only=false
    UseNonUniformScale=false
    UseOffset=true
    UseSphereOrientation=false
   )
   &InitialMeshOrientation(Orientation Vector=<0.3 1 0>, Use Orientation Vector=true, Use Rotation Vector=false)
  ]
  particleUpdateScript [
   &DynamicMaterialParameters(
    Index 0 Param 4=&FloatFromCurve(FloatCurve=DataInterface() {propertyValues {Curve {keys [<at 0 0>,<at 0.4 1>]}}})
    Module.Param0WriteEnabled=true
    Module.Param1WriteEnabled=true
    Module.Param2WriteEnabled=true
    Module.Param3WriteEnabled=true
   )
   &ParticleState()
   &Color(
    Color=<rgba 1 0.0833 0 1>
    Scale Alpha=&FloatFromCurve(FloatCurve=DataInterface() {propertyValues {Curve {keys [<at 0 1>,<at 1 0>]}}})
    Scale Color=&VectorFromCurve(VectorCurve=DataInterface() {propertyValues {XCurve {keys [<at -0.0031 48.8>]}, YCurve {keys [<at -0.006 20.3>]}, ZCurve {keys [<at -0.0011 1.0067>]}}})
   )
   &ScaleMeshSize(Scale Factor=&VectorFromFloat(Value=&FloatFromCurve(FloatCurve=DataInterface() {propertyValues {Curve {keys [<at 0 0.65>,<at 0.05 0.95>,<at 0.0907 0.9032>,<at 0.1427 0.9823>,<at 0.2 0.9021>,<at 0.2636 0.9927>,<at 0.3 0.9167>,<at 0.348 1.0178>,<at 0.4 0.8823>,<at 0.4416 1.037>,<at 0.5 0.9875>,<at 0.53 1.3626>,<at 0.56 1.1495>,<at 0.6 1.3078>,<at 1 1.3928>]}}})))
  ]
  rendererClasses [&NiagaraMeshRendererProperties]
  renderers [{rendererClass=&NiagaraMeshRendererProperties, rendererIndex=0}]
 }
 Glow {
  emitterUpdateScript [
   &EmitterState(Life Cycle Mode=::NewEnumerator1, Loop Behavior=::NewEnumerator1, Loop Duration=2, UseLoopDelay=true)
   &SpawnBurst_Instantaneous(Spawn Count=1, Spawn Time=1.96, Use Spawn Probability=false)
  ]
  particleSpawnScript [
   &InitializeParticle(
    Color=<rgba 1 0.5363 0.2077 0.566>
    Lifetime=0.7
    Sprite Size Mode=::NewEnumerator3
    Uniform Sprite Size=100
    UsePositionOffset=false
    Write Material Random=true
   ) {module=InitializeParticle001}
   &DynamicMaterialParameters(
    Index 0 Param 1=&RandomRangeFloat(Minimum=0, Maximum=1, Evaluation Type=::NewEnumerator1)
    Index 0 Param 2=1
    Index 0 Param 3=1
    Index 0 Param 4=1
    Module.Param0WriteEnabled=true
    Module.Param1WriteEnabled=false
    Module.Param2WriteEnabled=false
    Module.Param3WriteEnabled=false
   )
   &SphereLocation(
    MaskSpawnGroup=false
    Offset=<70.0881 120 124.424>
    Override Seed=false
    Sphere Radius=0.1
    Surface Only=false
    UseNonUniformScale=false
    UseOffset=true
    UseSphereOrientation=false
   )
  ]
  particleUpdateScript [
   &ParticleState()
   &DynamicMaterialParameters(
    Index 0 Param 2=1
    Index 0 Param 3=1
    Index 0 Param 4=1
    Module.Param0WriteEnabled=false
    Module.Param1WriteEnabled=true
    Module.Param2WriteEnabled=true
    Module.Param3WriteEnabled=true
   ) {module=DynamicMaterialParameters001}
   &ScaleColor(
    Scale Alpha=&FloatFromCurve(FloatCurve=DataInterface() {propertyValues {Curve {keys [<at 0 0 RCIM_Cubic>,<at 0.5 0.1875 RCIM_Cubic>,<at 1 0>]}}})
    ScaleA=true
    ScaleRGB=true
   )
   &ScaleSpriteSize(Scale Factor=&Vector2DFromFloat(Value=&FloatFromCurve(FloatCurve=DataInterface() {propertyValues {Curve {keys [
          {arriveTangent=3.3015, interpMode=RCIM_Cubic, leaveTangent=3.3015, tangentMode=RCTM_User, time=0, value=0}
          {arriveTangent=-0.0405, interpMode=RCIM_Cubic, leaveTangent=-0.0405, tangentMode=RCTM_User, time=1, value=1}
         ]}}})))
   &CameraOffset(Camera Offset Amount=55)
  ]
  rendererClasses [&NiagaraSpriteRendererProperties]
  renderers [{rendererClass=&NiagaraSpriteRendererProperties}]
 }
 Sparks_01 {
  emitterUpdateScript [
   &EmitterState() {module=EmitterState001}
   &SpawnBurst_Instantaneous(Spawn Count=20, Spawn Time=1.4, Use Spawn Probability=false)
  ]
  particleSpawnScript [
   &InitializeParticle(
    Color=<rgba 6 1.2914 0.5729 1>
    Lifetime Max=1
    Lifetime Min=0.6
    Lifetime Mode=::NewEnumerator1
    Sprite Size Mode=::NewEnumerator1
    Uniform Sprite Size Max=2
    Uniform Sprite Size Min=0.5
    UsePositionOffset=false
    Write Material Random=true
   ) {module=InitializeParticle001}
   &SphereLocation(
    MaskSpawnGroup=false
    Offset=<70 120 120>
    Override Seed=false
    Sphere Radius=0.01
    Surface Only=false
    UseNonUniformScale=false
    UseOffset=true
    UseSphereOrientation=false
   )
   &AddVelocityInCone(
    Cone Angle=60
    Cone Axis=<-0.36 0 1>
    Module.Use Velocity Falloff On Cone Axis=false
    Velocity Strength=&RandomRangeFloat(Minimum=117.9064, Maximum=374.6839)
   )
   &InheritVelocity(Module.Limit Speed=true, Velocity Limit=1000, Velocity Scale=<2 2 2>) {enabled=false}
  ]
  particleUpdateScript [
   &ParticleState()
   &Drag(Drag=2.3032, Use Linear Drag=true, Use Rotational Drag=false)
   &GravityForce(Gravity=<0 0 -500>)
   &SolveForcesAndVelocity(Clamp Velocity=false, Limit Acceleration=false)
   &ScaleColor(
    Scale Alpha=&FloatFromCurve(FloatCurve=DataInterface() {propertyValues {Curve {keys [<at 0 1>,<at 1 0>]}}})
    Scale RGB=&VectorFromFloat(Value=&FloatFromCurve(FloatCurve=DataInterface() {propertyValues {Curve {keys [<at 0 50>,<at 0.3 35>,<at 1 0>]}}}))
    ScaleA=true
    ScaleRGB=true
   )
   &ScaleSpriteSize(Scale Factor=&Vector2DFromFloat(Value=&FloatFromCurve(FloatCurve=DataInterface() {propertyValues {Curve {keys [
          <at 0 0.15 RCIM_Cubic>
          {interpMode=RCIM_Cubic, tangentMode=RCTM_User, time=0.3, value=1.5}
          {interpMode=RCIM_Cubic, tangentMode=RCTM_User, time=1, value=0}
         ]}}})))
  ]
  rendererClasses [&NiagaraSpriteRendererProperties]
  renderers [{rendererClass=&NiagaraSpriteRendererProperties}]
 }
 Sparks_02 {
  emitterUpdateScript [
   &EmitterState() {module=EmitterState001}
   &SpawnBurst_Instantaneous(Spawn Count=20, Spawn Time=1.6, Use Spawn Probability=false)
  ]
  particleSpawnScript [
   &InitializeParticle(
    Color=<rgba 6 1.2914 0.5729 1>
    Lifetime Max=1
    Lifetime Min=0.6
    Lifetime Mode=::NewEnumerator1
    Sprite Size Mode=::NewEnumerator1
    Uniform Sprite Size Max=2
    Uniform Sprite Size Min=0.5
    UsePositionOffset=false
    Write Material Random=true
   ) {module=InitializeParticle001}
   &SphereLocation(
    MaskSpawnGroup=false
    Offset=<70.0881 120 112.7408>
    Override Seed=false
    Sphere Radius=0.01
    Surface Only=false
    UseNonUniformScale=false
    UseOffset=true
    UseSphereOrientation=false
   )
   &AddVelocityInCone(
    Cone Angle=60
    Cone Axis=<0.23 0 1>
    Module.Use Velocity Falloff On Cone Axis=false
    Velocity Strength=&RandomRangeFloat(Minimum=117.9064, Maximum=374.6839)
   )
   &InheritVelocity(Module.Limit Speed=true, Velocity Limit=1000, Velocity Scale=<2 2 2>) {enabled=false}
  ]
  particleUpdateScript [
   &ParticleState()
   &Drag(Drag=2.3032, Use Linear Drag=true, Use Rotational Drag=false)
   &GravityForce(Gravity=<0 0 -500>)
   &SolveForcesAndVelocity(Clamp Velocity=false, Limit Acceleration=false)
   &ScaleColor(
    Scale Alpha=&FloatFromCurve(FloatCurve=DataInterface() {propertyValues {Curve {keys [<at 0 1>,<at 1 0>]}}})
    Scale RGB=&VectorFromFloat(Value=&FloatFromCurve(FloatCurve=DataInterface() {propertyValues {Curve {keys [<at 0 50>,<at 0.3 35>,<at 1 0>]}}}))
    ScaleA=true
    ScaleRGB=true
   )
   &ScaleSpriteSize(Scale Factor=&Vector2DFromFloat(Value=&FloatFromCurve(FloatCurve=DataInterface() {propertyValues {Curve {keys [
          <at 0 0.15 RCIM_Cubic>
          {interpMode=RCIM_Cubic, tangentMode=RCTM_User, time=0.3, value=1.5}
          {interpMode=RCIM_Cubic, tangentMode=RCTM_User, time=1, value=0}
         ]}}})))
  ]
  rendererClasses [&NiagaraSpriteRendererProperties]
  renderers [{rendererClass=&NiagaraSpriteRendererProperties}]
 }
 Sparks_03 {
  emitterUpdateScript [
   &EmitterState() {module=EmitterState001}
   &SpawnBurst_Instantaneous(Spawn Count=20, Spawn Time=1.7, Use Spawn Probability=false)
  ]
  particleSpawnScript [
   &InitializeParticle(
    Color=<rgba 6 1.2914 0.5729 1>
    Lifetime Max=1
    Lifetime Min=0.6
    Lifetime Mode=::NewEnumerator1
    Sprite Size Mode=::NewEnumerator1
    Uniform Sprite Size Max=2
    Uniform Sprite Size Min=0.5
    UsePositionOffset=false
    Write Material Random=true
   ) {module=InitializeParticle001}
   &SphereLocation(
    MaskSpawnGroup=false
    Offset=<70.0881 120 112.7408>
    Override Seed=false
    Sphere Radius=0.01
    Surface Only=false
    UseNonUniformScale=false
    UseOffset=true
    UseSphereOrientation=false
   )
   &AddVelocityInCone(
    Cone Angle=60
    Cone Axis=<-0.01 0 1>
    Module.Use Velocity Falloff On Cone Axis=false
    Velocity Strength=&RandomRangeFloat(Minimum=129.0745, Maximum=505.0443)
   )
   &InheritVelocity(Module.Limit Speed=true, Velocity Limit=1000, Velocity Scale=<2 2 2>) {enabled=false}
  ]
  particleUpdateScript [
   &ParticleState()
   &Drag(Drag=2.3032, Use Linear Drag=true, Use Rotational Drag=false)
   &GravityForce(Gravity=<0 0 -500>)
   &SolveForcesAndVelocity(Clamp Velocity=false, Limit Acceleration=false)
   &ScaleColor(
    Scale Alpha=&FloatFromCurve(FloatCurve=DataInterface() {propertyValues {Curve {keys [<at 0 1>,<at 1 0>]}}})
    Scale RGB=&VectorFromFloat(Value=&FloatFromCurve(FloatCurve=DataInterface() {propertyValues {Curve {keys [<at 0 50>,<at 0.3 35>,<at 1 0>]}}}))
    ScaleA=true
    ScaleRGB=true
   )
   &ScaleSpriteSize(Scale Factor=&Vector2DFromFloat(Value=&FloatFromCurve(FloatCurve=DataInterface() {propertyValues {Curve {keys [
          <at 0 0.15 RCIM_Cubic>
          {interpMode=RCIM_Cubic, tangentMode=RCTM_User, time=0.3, value=1.5}
          {interpMode=RCIM_Cubic, tangentMode=RCTM_User, time=1, value=0}
         ]}}})))
  ]
  rendererClasses [&NiagaraSpriteRendererProperties]
  renderers [{rendererClass=&NiagaraSpriteRendererProperties}]
 }
 Sparks_04 {
  emitterUpdateScript [
   &EmitterState() {module=EmitterState001}
   &SpawnBurst_Instantaneous(Spawn Count=20, Spawn Time=2, Use Spawn Probability=false)
  ]
  particleSpawnScript [
   &InitializeParticle(
    Color=<rgba 6 1.2914 0.5729 1>
    Lifetime Max=1
    Lifetime Min=0.6
    Lifetime Mode=::NewEnumerator1
    Sprite Size Mode=::NewEnumerator1
    Uniform Sprite Size Max=2
    Uniform Sprite Size Min=0.5
    UsePositionOffset=false
    Write Material Random=true
   ) {module=InitializeParticle001}
   &SphereLocation(
    MaskSpawnGroup=false
    Offset=<70.0881 120 112.7408>
    Override Seed=false
    Sphere Radius=20
    Surface Only=false
    UseNonUniformScale=false
    UseOffset=true
    UseSphereOrientation=false
   )
   &InheritVelocity(Module.Limit Speed=true, Velocity Limit=1000, Velocity Scale=<2 2 2>) {enabled=false}
   &AddVelocity(Velocity=&RandomRangeVector(Minimum=<-200 -200 -100>, Maximum=<200 200 400>))
  ]
  particleUpdateScript [
   &ParticleState()
   &Drag(Drag=2.3032, Use Linear Drag=true, Use Rotational Drag=false)
   &GravityForce(Gravity=<0 0 -500>)
   &SolveForcesAndVelocity(Clamp Velocity=false, Limit Acceleration=false)
   &ScaleColor(
    Scale Alpha=&FloatFromCurve(FloatCurve=DataInterface() {propertyValues {Curve {keys [<at 0 1>,<at 1 0>]}}})
    Scale RGB=&VectorFromFloat(Value=&FloatFromCurve(FloatCurve=DataInterface() {propertyValues {Curve {keys [<at 0 50>,<at 0.3 35>,<at 1 0>]}}}))
    ScaleA=true
    ScaleRGB=true
   )
   &ScaleSpriteSize(Scale Factor=&Vector2DFromFloat(Value=&FloatFromCurve(FloatCurve=DataInterface() {propertyValues {Curve {keys [
          <at 0 0.15 RCIM_Cubic>
          {interpMode=RCIM_Cubic, tangentMode=RCTM_User, time=0.3, value=1.5}
          {interpMode=RCIM_Cubic, tangentMode=RCTM_User, time=1, value=0}
         ]}}})))
  ]
  rendererClasses [&NiagaraSpriteRendererProperties]
  renderers [{rendererClass=&NiagaraSpriteRendererProperties}]
 }
}
summary {
 systemName=NS_Emote_FirecrackerFish
 emitters [
  {emitterName=Fish_Shape, rendererClasses [&NiagaraMeshRendererProperties]}
  {emitterName=Sparks_01, rendererClasses [&NiagaraSpriteRendererProperties]}
  {emitterName=Sparks_02, rendererClasses [&NiagaraSpriteRendererProperties]}
  {emitterName=Sparks_03, rendererClasses [&NiagaraSpriteRendererProperties]}
  {emitterName=Sparks_04, rendererClasses [&NiagaraSpriteRendererProperties]}
  {emitterName=Glow, rendererClasses [&NiagaraSpriteRendererProperties]}
 ]
 userVariables [
  User.bIsFrontEnd {type=NiagaraBool}
  User.bIsFrontEndPreview {type=NiagaraBool}
  User.FrontendSpawn {type=NiagaraBool}
  User.RibbonIDOffset {type=NiagaraInt32}
  User.SkeletalMesh {
   type=NiagaraDataInterfaceSkeletalMesh
   defaultValue {
    struct=NiagaraExt_VariableValue_DataInterface
    value {dataInterface=&self:NiagaraDataInterfaceSkeletalMesh_1}
   }
  }
 ]
}
