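# Niagara effect for the FirecrackerSparks player emote, drawing a continuous burst of firecracker sparks.
# Embers, Sparkles and gravity-dragged Sparks_Gravity sprites shoot upward in cones from a small shared offset, around a flickering mesh Flame.
# Glow adds pulsing sprites with lights, and Light001 pairs a component renderer with soft sprites.
# AnimTrail_Oil draws a short-lived ribbon using custom AnimTrail modules, with skeletal sampling filtered to the weapon_l bone.
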NiagaraSystem /BRCosmetics/Animation/Game/MainPlayer/Emotes/FirecrackerSparks/FX/NS_Emote_FirecrackerSpark
dependencies {
 usedDataInterfaces [
  &NiagaraDataInterfaceCurve
  &NiagaraDataInterfaceColorCurve
  &NiagaraDataInterfaceVector2DCurve
  &NiagaraDataInterfaceVectorCurve
 ]
 usedDynamicInputs []
 usedModules [&SystemState]
}
emitters {
 AnimTrail_Oil {
  emitterUpdateScript [
   &EmitterState()
   &FN_SpawnAnimTrail_v2(AnimTrailWidthMode=::FromSecond, Spawn Mode:ENiagaraSkeletalSpawning=::NewEnumerator0)
  ]
  particleSpawnScript [
   &SphereLocation(
    MaskSpawnGroup=false
    Offset=<1 0.5 100>
    Override Seed=false
    Sphere Radius=0.01
    Surface Only=false
    UseNonUniformScale=false
    UseOffset=true
    UseSphereOrientation=false
   ) {enabled=false}
   /Game/Effects/Niagara/Modules/AnimTrail/FN_InitializeAnimTrail/FN_InitializeAnimTrailBoneIndecies(AnimTrail Selection=1)
   /Game/Effects/Niagara/Modules/AnimTrail/FN_InitializeAnimTrail/FN_InitializeAnimTrail(AnimTrailWidth=0.07, Lifetime=0.25)
   /SetVariables(Particles.RibbonU1Override:NiagaraFloat=&FloatFromCurve(FloatCurve=DataInterface() {propertyValues {Curve {keys [<at 0 1>,<at 1 12>]}}}, CurveIndex=$Emitter.NormalizedLoopAge)) {isSetParameters=true}
   &AddVelocity(Velocity=<0 0 1>)
   &SystemLocation(MaskSpawnGroup=false, Offset=<0 0 62>)
  ]
  particleUpdateScript [
   /Game/Effects/Niagara/Modules/Ribbons/FN_UpdateRibbon_v6(
    Enable Skeletal Position=true
    SkeletalMesh=DataInterface() {propertyValues {
      DefaultMesh=null
      ExcludeBoneName=root
      OptionalComponentClass=null
      PreviewMesh=/Game/Characters/Player/Male/Medium/Bodies/M_Med_Soldier_04/Meshes/SK_M_Med_Soldier_04
      SkinningMode=SkinOnTheFly
      SoftSourceActor=null
      SourceMode=Default
      UvSetIndex=0
      WholeMeshLOD=-1
      bExcludeBone=false
      bReadDeformedGeometry=true
      bRequireCurrentFrameData=true
      ComponentTags []
      FilteredBones [weapon_l]
      FilteredSockets []
      MeshUserParameter {parameter {name=None, typeDefHandle {registeredTypeIndex=1}, varData []}}
      SamplingRegions []
     }}
   ) {enabled=false}
   &ScaleRibbonWidth(Ribbon Width Scale=&FloatFromCurve(FloatCurve=DataInterface() {propertyValues {Curve {keys [
         <at 0 0.1>
         {arriveTangent=-3.1505, interpMode=RCIM_Cubic, leaveTangent=-3.1505, tangentMode=RCTM_User, time=0.1, value=0.85}
         {arriveTangent=0.0357, interpMode=RCIM_Cubic, leaveTangent=0.0357, tangentMode=RCTM_User, time=1, value=0}
        ]}}}))
   &FN_UpdateAnimTrail(SkeletalMesh=DataInterface() {propertyValues {
      DefaultMesh=null
      ExcludeBoneName=root
      OptionalComponentClass=null
      PreviewMesh=&M_Med_Soldier_04_Preview
      SkinningMode=SkinOnTheFly
      SoftSourceActor=null
      SourceMode=Default
      UvSetIndex=0
      WholeMeshLOD=-1
      bExcludeBone=false
      bReadDeformedGeometry=true
      bRequireCurrentFrameData=true
      ComponentTags []
      FilteredBones [weapon_l]
      FilteredSockets []
      MeshUserParameter {parameter {name=None, typeDefHandle {registeredTypeIndex=1}, varData []}}
      SamplingRegions []
     }})
   &SolveForcesAndVelocity(Clamp Velocity=false, Limit Acceleration=false)
   &FN_UpdateAnimTrail_v2() {enabled=false}
   &DynamicMaterialParameters(
    Index 0 Param 3=1
    Module.Param0WriteEnabled=true
    Module.Param1WriteEnabled=true
    Module.Param2WriteEnabled=true
    Module.Param3WriteEnabled=true
   )
  ]
  rendererClasses [&NiagaraRibbonRendererProperties]
  renderers [{rendererClass=&NiagaraRibbonRendererProperties}]
 }
 Embers {
  emitterUpdateScript [
   &EmitterState(
    Enable Distance Culling=true
    Enable Visibility Culling=true
    MaxDistance=1000
    Module.bApplySpawnCountScaleByDistance=true
    Module.bUseMaxDistance=true
    Module.bUseMinDistance=false
    Scalability Mode=::NewEnumerator1
    Scale Spawn Count=true
    Spawn Count Scale By Distance=DataInterface() {propertyValues {Curve {keys [<at 0 1>,<at 0.3 1>,<at 1 0>]}}}
   )
   &SpawnRate(SpawnRate=40, Use Spawn Probability=false) {module=SpawnRate001}
  ]
  particleSpawnScript [
   &InitializeParticle(
    Color=<rgba 5 4.7777 3.4057 1>
    Lifetime Max=0.25
    Lifetime Min=0.15
    Lifetime Mode=::NewEnumerator1
    Sprite Size Mode=::NewEnumerator1
    Uniform Sprite Size Max=3
    Uniform Sprite Size Min=1
    UsePositionOffset=false
    Write Material Random=true
   ) {module=InitializeParticle001}
   &SphereLocation(
    MaskSpawnGroup=false
    Offset=<1 0.5 61>
    Override Seed=false
    Sphere Radius=0.1
    Surface Only=false
    UseNonUniformScale=false
    UseOffset=true
    UseSphereOrientation=false
   ) {module=SphereLocation001}
   &AddVelocityFromPoint(
    Module.Use Falloff Distance=false
    Origin Offset=<5 0 -20>
    Velocity Strength=&RandomRangeFloat(Minimum=200, Maximum=250)
   ) {enabled=false}
   &AddVelocityInCone(
    Cone Angle=75
    Cone Axis=<0 0 1>
    Cone Axis Coordinate Space=::Simulation
    Module.Use Velocity Falloff On Cone Axis=false
    Velocity Distribution Along Cone Axis=0.6
    Velocity Strength=&RandomRangeFloat(Minimum=200, Maximum=300)
   )
  ]
  particleUpdateScript [
   &ParticleState()
   &SolveForcesAndVelocity(Clamp Velocity=false, Limit Acceleration=false)
   &ScaleColor(
    Scale Alpha=&FloatFromCurve(FloatCurve=DataInterface() {propertyValues {Curve {keys [<at 0.5 0.95>,<at 1 0>]}}})
    Scale RGB=&MakeVectorFromLinearColorRGB(Color=&ColorFromCurve(ColorCurve=DataInterface() {propertyValues {
        AlphaCurve {keys [<at 0 1 RCIM_Cubic>,<at 0.7107 1>]}
        BlueCurve {keys [<at 0 14.6701 RCIM_Cubic>,<at 0.2499 1.5104>,<at 0.7114 0.1753 RCIM_Cubic>]}
        GreenCurve {keys [<at 0 36.7481 RCIM_Cubic>,<at 0.2499 5.293>,<at 0.7114 0.3243 RCIM_Cubic>]}
        RedCurve {keys [<at 0 50 RCIM_Cubic>,<at 0.2499 10>,<at 0.7114 1 RCIM_Cubic>]}
       }}))
    ScaleA=true
    ScaleRGB=true
   )
   &ScaleSpriteSize(Scale Factor=&Vector2DFromCurve(
     Vector2Curve=DataInterface() {propertyValues {
       XCurve {keys [
         {arriveTangent=0.9446, interpMode=RCIM_Cubic, leaveTangent=0.9446, tangentMode=RCTM_User, time=0, value=0.625}
         {interpMode=RCIM_Cubic, tangentMode=RCTM_User, time=0.4, value=0.5}
         <at 1 0>
        ]}
       YCurve {keys [
         {arriveTangent=0.9446, interpMode=RCIM_Cubic, leaveTangent=0.9446, tangentMode=RCTM_User, time=0, value=3.5}
         {arriveTangent=-11.1745, interpMode=RCIM_Cubic, leaveTangent=-11.1745, tangentMode=RCTM_User, time=0.4, value=2.25}
         <at 1 0>
        ]}
      }}
     ScaleCurve=&Vector2DFromFloat(Value=&FN_LOD_Distance_Remap(Start Distance=300, End Distance=1000, Minimum Output=1, Maximum Output=1.1))
    ))
  ]
  rendererClasses [&NiagaraSpriteRendererProperties]
  renderers [{rendererClass=&NiagaraSpriteRendererProperties}]
 }
 Flame {
  emitterUpdateScript [
   &EmitterState()
   &SpawnRate(SpawnRate=40, Use Spawn Probability=false)
   &SpawnBurst_Instantaneous(Spawn Count=4, Use Spawn Probability=false) {enabled=false}
  ]
  particleSpawnScript [
   &InitializeParticle(
    Lifetime Max=0.25
    Lifetime Min=0.15
    Lifetime Mode=::NewEnumerator1
    Mesh Scale Max=<0.1 0.2 0.2>
    Mesh Scale Min=<0.01 0.05 0.05>
    Mesh Scale Mode=::NewEnumerator2
    UsePositionOffset=false
    Write Material Random=true
   ) {module=InitializeParticle001}
   /Game/Effects/Niagara/Modules/Orientation/FN_Fan(
    Cone Angle:NiagaraFloat=&UniformRangedFloat(Minimum=0, Maximum=0)
    Spin:NiagaraFloat=&UniformRangedFloat(Minimum=0, Maximum=6.2)
   )
   &SphereLocation(
    MaskSpawnGroup=false
    Offset=<1 0.5 61>
    Override Seed=false
    Sphere Radius=0.01
    Surface Only=false
    UseNonUniformScale=false
    UseOffset=true
    UseSphereOrientation=false
   )
   &InitialMeshOrientation(
    Orientation Vector=&RandomRangeVector(Minimum=<0.1 0.1 1>, Maximum=<-0.1 -0.1 1>)
    Use Orientation Vector=true
    Use Rotation Vector=false
   )
  ]
  particleUpdateScript [
   &ParticleState()
   &Color(
    Color=<rgba 1 0.0833 0 1>
    Scale Color=&VectorFromCurve(VectorCurve=DataInterface() {propertyValues {
       XCurve {keys [<at -0.0031 48.8>,<at 1 1>]}
       YCurve {keys [<at -0.006 20.3>,<at 1 0.8>]}
       ZCurve {keys [<at -0.0011 1.0067>,<at 0.9989 0.28>]}
      }})
   )
   &ScaleMeshSize(Scale Factor=&VectorFromCurve(VectorCurve=DataInterface() {propertyValues {
       XCurve {keys [<at 0 0.55>,<at 0.2 2.875>,<at 1 0>]}
       YCurve {keys [<at 0 0.05>,<at 0.2 2.5>,<at 1 0>]}
       ZCurve {keys [<at 0 0.55>,<at 0.2 2>,<at 1 0>]}
      }}))
  ]
  rendererClasses [&NiagaraMeshRendererProperties]
  renderers [{rendererClass=&NiagaraMeshRendererProperties, rendererIndex=0}]
 }
 Glow {
  emitterUpdateScript [
   &EmitterState()
   &SpawnBurst_Instantaneous(Spawn Count=1, Use Spawn Probability=false) {enabled=false}
   &SpawnRate(SpawnRate=8, Use Spawn Probability=false)
  ]
  particleSpawnScript [
   &InitializeParticle(
    Color=<rgba 1 0.1371 0.0475 1>
    Lifetime=0.4
    Sprite Rotation Mode=::NewEnumerator2
    Sprite Size Mode=::NewEnumerator1
    Uniform Sprite Size Max=60
    Uniform Sprite Size Min=30
    UsePositionOffset=false
    Write Material Random=true
   ) {module=InitializeParticle001}
   &DynamicMaterialParameters(
    Index 0 Param 1=0.5
    Module.Param0WriteEnabled=true
    Module.Param1WriteEnabled=true
    Module.Param2WriteEnabled=true
    Module.Param3WriteEnabled=true
   )
   &SphereLocation(
    MaskSpawnGroup=false
    Offset=<1 0.5 61>
    Override Seed=false
    Sphere Radius=0.01
    Surface Only=false
    UseNonUniformScale=false
    UseOffset=true
    UseSphereOrientation=false
   )
  ]
  particleUpdateScript [
   &ParticleState()
   &SolveForcesAndVelocity(Clamp Velocity=false, Limit Acceleration=false)
   &ScaleColor(
    Scale RGB=&VectorFromFloat(Value=&FloatFromCurve(
      FloatCurve=DataInterface() {propertyValues {Curve {keys [{interpMode=RCIM_Cubic, tangentMode=RCTM_User, time=0, value=0}, <at 0.5 1 RCIM_Cubic>, <at 1 0>]}}}
      Scale Curve=1.5
     ))
    ScaleA=true
    ScaleRGB=true
   )
   /SetVariables(Particles.LightColor=&ScaleLinearColorByCurve(
     InitialValue=<rgba 1 0.5961 0.2471 1>
     ScaleCurve=DataInterface() {propertyValues {
       AlphaCurve {keys [{interpMode=RCIM_Cubic, tangentMode=RCTM_User, time=0, value=0}, {interpMode=RCIM_Cubic, tangentMode=RCTM_User, time=0.5, value=5}, <at 1 0>]}
       BlueCurve {keys [<at 0 1>]}
       GreenCurve {keys [<at 0 1>]}
       RedCurve {keys [<at 0 1>]}
      }}
     CurveIndex=$Particles.NormalizedAge
    )) {isSetParameters=true}
   &ScaleSpriteSize(Scale Factor=&Vector2DFromFloat(Value=&FloatFromCurve(FloatCurve=DataInterface() {propertyValues {Curve {keys [
          {interpMode=RCIM_Cubic, tangentMode=RCTM_User, time=0, value=0}
          {interpMode=RCIM_Cubic, tangentMode=RCTM_User, time=0.4, value=1}
          {interpMode=RCIM_Cubic, tangentMode=RCTM_User, time=1, value=0}
         ]}}})))
   &CameraOffset(Camera Offset Amount=-30)
  ]
  rendererClasses [&NiagaraSpriteRendererProperties, &NiagaraLightRendererProperties]
  renderers [
   {rendererClass=&NiagaraSpriteRendererProperties}
   {rendererClass=&NiagaraLightRendererProperties, rendererIndex=1}
  ]
 }
 Light001 {
  emitterUpdateScript [
   &EmitterState()
   &SpawnBurst_Instantaneous(Spawn Count=1, Use Spawn Probability=false) {enabled=false}
   &SpawnRate(SpawnRate=4, Use Spawn Probability=false)
  ]
  particleSpawnScript [
   &InitializeParticle(
    Lifetime=0.7
    Sprite Rotation Mode=::NewEnumerator2
    Sprite Size Mode=::NewEnumerator1
    Uniform Sprite Size Max=200
    Uniform Sprite Size Min=100
    UsePositionOffset=false
    Write Material Random=true
   ) {module=InitializeParticle001}
   &DynamicMaterialParameters(
    Index 0 Param 1=0.5
    Module.Param0WriteEnabled=true
    Module.Param1WriteEnabled=true
    Module.Param2WriteEnabled=true
    Module.Param3WriteEnabled=true
   )
   &SphereLocation(
    MaskSpawnGroup=false
    Offset=<6 -5 65>
    Override Seed=false
    Sphere Radius=0.01
    Surface Only=false
    UseNonUniformScale=false
    UseOffset=true
    UseSphereOrientation=false
   )
  ]
  particleUpdateScript [
   &ParticleState(Kill Particles When Lifetime Has Elapsed=false, Let Infinitely Lived Particles Die When Emitter Deactivates=true)
   &SolveForcesAndVelocity(Clamp Velocity=false, Limit Acceleration=false)
   &ScaleColor(
    Scale RGB=&VectorFromFloat(Value=&FloatFromCurve(
      FloatCurve=DataInterface() {propertyValues {Curve {keys [{interpMode=RCIM_Cubic, tangentMode=RCTM_User, time=0, value=0}, <at 0.5 1 RCIM_Cubic>, <at 1 0>]}}}
      Scale Curve=1.5
     ))
    ScaleA=true
    ScaleRGB=true
   )
   &ScaleSpriteSize(Scale Factor=&Vector2DFromFloat(Value=&FloatFromCurve(FloatCurve=DataInterface() {propertyValues {Curve {keys [
          {interpMode=RCIM_Cubic, tangentMode=RCTM_User, time=0, value=0}
          {interpMode=RCIM_Cubic, tangentMode=RCTM_User, time=0.4, value=1}
          {interpMode=RCIM_Cubic, tangentMode=RCTM_User, time=1, value=0}
         ]}}})))
   &CameraOffset(Camera Offset Amount=-30)
  ]
  rendererClasses [&NiagaraComponentRendererProperties, &NiagaraSpriteRendererProperties]
  renderers [
   {rendererClass=&NiagaraComponentRendererProperties, rendererIndex=0}
   {rendererClass=&NiagaraSpriteRendererProperties, rendererIndex=1}
  ]
 }
 Sparkles {
  emitterUpdateScript [
   &EmitterState()
   &SpawnRate(Spawn Probability=0.4, SpawnRate=25, Use Spawn Probability=true)
  ]
  particleSpawnScript [
   &InitializeParticle(
    Color=<rgba 20 9.9456 3.3946 1>
    Lifetime=&RandomRangeFloat(Minimum=0.15, Maximum=0.25)
    Sprite Rotation Angle=&RandomRangeFloat(Minimum=-0.02, Maximum=0.02)
    Sprite Rotation Mode=::NewEnumerator1
    Sprite Size Mode=::NewEnumerator1
    Uniform Sprite Size Min=2
    UsePositionOffset=false
    Write Material Random=true
   )
   &SphereLocation(
    MaskSpawnGroup=false
    Offset=<1 0.5 61>
    Override Seed=false
    Sphere Radius=0.01
    Surface Only=false
    UseNonUniformScale=false
    UseOffset=true
    UseSphereOrientation=false
   ) {module=SphereLocation001}
   &AddVelocityInCone(
    Cone Angle=40
    Cone Axis=<0 0 1>
    Cone Axis Coordinate Space=::Simulation
    Module.Use Velocity Falloff On Cone Axis=false
    Velocity Strength=&RandomRangeFloat(Minimum=240, Maximum=220)
   )
  ]
  particleUpdateScript [
   &SolveForcesAndVelocity(Clamp Velocity=false, Limit Acceleration=false)
   &ParticleState()
   &ScaleSpriteSize(Scale Factor=&Vector2DFromFloat(Value=&FloatFromCurve(FloatCurve=DataInterface() {propertyValues {Curve {keys [<at 0 0 RCIM_Cubic>,<at 0.5 1 RCIM_Cubic>,<at 1 0 RCIM_Cubic>]}}})))
   &ScaleColor(
    Scale Alpha=&FloatFromCurve(FloatCurve=DataInterface() {propertyValues {Curve {keys [{arriveTangent=-1.402, interpMode=RCIM_Cubic, leaveTangent=-1.402, tangentMode=RCTM_User, time=0, value=1}, <at 0.13 1>, <at 1 0 RCIM_Cubic>]}}})
    ScaleA=true
    ScaleRGB=true
   )
   &CameraOffset(Camera Offset Amount=5)
  ]
  rendererClasses [&NiagaraSpriteRendererProperties]
  renderers [{rendererClass=&NiagaraSpriteRendererProperties}]
 }
 Sparks_Gravity {
  emitterUpdateScript [
   &EmitterState() {module=EmitterState001}
   &SpawnRate(Spawn Probability=0.4, SpawnRate=40, Use Spawn Probability=false)
   /SetVariables(Emitter.VelocityPosition:Vector3f=&RandomRangeVector(Minimum=<-20 -20 -20>, Maximum=<20 20 20>, Evaluation Type=::NewEnumerator1, Random Seed=$Emitter.LoopCount, Fixed Random Seed=true)) {enabled=false, isSetParameters=true}
  ]
  particleSpawnScript [
   /Niagara/Modules/Spawn/Initialization/InitializeParticle(
    Color=<rgba 6 1.2914 0.5729 1>
    Lifetime=&UniformRangedFloat(Minimum=0.4, Maximum=0.6)
    Module.Write Color=true
    Module.Write Lifetime=true
    Module.Write Mass=true
    Module.Write Material Random=true
    Module.Write Position=false
    Module.Write Scale=false
    Module.Write SpriteRotation=false
    Module.Write SpriteSize=true
    Position=&SimulationPosition()
    Sprite Size=&RandomRangeVector2D(
     Minimum=&Vector2DFromFloat(Value=0.3)
     Maximum=&Vector2DFromFloat(Value=1)
    )
   )
   &SphereLocation(
    MaskSpawnGroup=false
    Offset=<1 0.5 61>
    Override Seed=false
    Sphere Radius=0.01
    Surface Only=false
    UseNonUniformScale=false
    UseOffset=true
    UseSphereOrientation=false
   ) {module=SphereLocation001}
   &AddVelocityInCone(
    Cone Angle=60
    Cone Axis=<0 0 1>
    Module.Use Velocity Falloff On Cone Axis=false
    Velocity Strength=&RandomRangeFloat(Minimum=250, Maximum=400)
   )
   &InheritVelocity(Module.Limit Speed=true, Velocity Limit=1000, Velocity Scale=<2 2 2>)
  ]
  particleUpdateScript [
   &ParticleState()
   &Drag(Drag=8, Use Linear Drag=true, Use Rotational Drag=false)
   &GravityForce(Gravity=<0 0 -500>)
   &SolveForcesAndVelocity(Clamp Velocity=false, Limit Acceleration=false)
   &ScaleColor(
    Scale Alpha=&FloatFromCurve(FloatCurve=DataInterface() {propertyValues {Curve {keys [<at 0 1>,<at 1 0>]}}})
    Scale RGB=&VectorFromFloat(Value=&FloatFromCurve(FloatCurve=DataInterface() {propertyValues {Curve {keys [<at 0 50>,<at 0.3 35>,<at 1 0>]}}}))
    ScaleA=true
    ScaleRGB=true
   )
   &ScaleSpriteSize(Scale Factor=&Vector2DFromFloat(Value=&FloatFromCurve(FloatCurve=DataInterface() {propertyValues {Curve {keys [
          <at 0 0.15 RCIM_Cubic>
          {interpMode=RCIM_Cubic, tangentMode=RCTM_User, time=0.3, value=1.5}
          {interpMode=RCIM_Cubic, tangentMode=RCTM_User, time=1, value=0}
         ]}}})))
  ]
  rendererClasses [&NiagaraSpriteRendererProperties]
  renderers [{rendererClass=&NiagaraSpriteRendererProperties}]
 }
}
summary {
 systemName=NS_Emote_FirecrackerSpark
 emitters [
  {emitterName=Embers, rendererClasses [&NiagaraSpriteRendererProperties]}
  {emitterName=Sparkles, rendererClasses [&NiagaraSpriteRendererProperties]}
  {emitterName=Sparks_Gravity, rendererClasses [&NiagaraSpriteRendererProperties]}
  {emitterName=Flame, rendererClasses [&NiagaraMeshRendererProperties]}
  {emitterName=Light001, rendererClasses [&NiagaraComponentRendererProperties, &NiagaraSpriteRendererProperties]}
  {emitterName=Glow, rendererClasses [&NiagaraSpriteRendererProperties, &NiagaraLightRendererProperties]}
  {emitterName=AnimTrail_Oil, rendererClasses [&NiagaraRibbonRendererProperties]}
 ]
 userVariables [
  User.FrontendSpawn {type=NiagaraBool}
  User.RibbonIDOffset {type=NiagaraInt32}
  User.SkeletalMesh {
   type=NiagaraDataInterfaceSkeletalMesh
   defaultValue {
    struct=NiagaraExt_VariableValue_DataInterface
    value {dataInterface=&self:NiagaraDataInterfaceSkeletalMesh_1}
   }
  }
 ]
}
